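# Physics asset for the EssayViewPier backpack cosmetic, defining collision bodies and joints for a chain of tape bones.
# A small capsule on base_tape anchors four identical, offset capsules on dyn_tape_001 through dyn_tape_004.
# Each segment is constrained to the previous one, with tighter swing at the root joint, looser swing further down and almost no twist throughout.
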PhysicsAsset /BRCosmetics/Accessories/FORT_Backpacks/Backpack_M_MED_EssayViewPier/Meshes/M_MED_EssayViewPier_Pack_Physics
bodies {
 base_tape [{shapeType=Capsule, length=0.505, radius=0.505, shapeName=base_tape_capsule, rotation=<rot 0 -90 0>}]
 dyn_tape_001 [{shapeType=Capsule, length=2, radius=1, shapeName=dyn_tape_001_capsule, center=<2 0 0>, rotation=<rot 90 0 0>}]
 dyn_tape_002 [{shapeType=Capsule, length=2, radius=1, shapeName=dyn_tape_002_capsule, center=<2 0 0>, rotation=<rot 90 0 0>}]
 dyn_tape_003 [{shapeType=Capsule, length=2, radius=1, shapeName=dyn_tape_003_capsule, center=<2 0 0>, rotation=<rot 90 0 0>}]
 dyn_tape_004 [{shapeType=Capsule, length=2, radius=1, shapeName=dyn_tape_004_capsule, center=<2 0 0>, rotation=<rot 90 0 0>}]
}
constraints [
 {bone1Name=dyn_tape_002, bone2Name=dyn_tape_001, swing1LimitDegrees=25, swing2LimitDegrees=25, twistLimitDegrees=1}
 {bone1Name=dyn_tape_003, bone2Name=dyn_tape_002, swing1LimitDegrees=25, swing2LimitDegrees=25, twistLimitDegrees=1}
 {bone1Name=dyn_tape_004, bone2Name=dyn_tape_003, swing1LimitDegrees=25, swing2LimitDegrees=25, twistLimitDegrees=1}
 {bone1Name=dyn_tape_001, bone2Name=base_tape, swing1LimitDegrees=10, swing2LimitDegrees=10, twistLimitDegrees=1}
]
