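# Animation Blueprint for the EssayViewPier backpack cosmetic, playing an idle with a timed random idle break and adding secondary motion.
# An additive rotation driven by the owner's spine yaw offsets the pack bone, then AnimDynamics swings base_chair and five caster bones before a RigidBody pass.
# A deprecated gravity-override function feeds base_chair gravity and RigidBody external force.
# EventGraph also sets backpack translation and front-end alpha and simulation-space variables, which the AnimGraph leaves unused.
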AnimBlueprint /BRCosmetics/Accessories/FORT_Backpacks/Backpack_M_MED_EssayViewPier/Meshes/M_MED_EssayViewPier_Pack_AnimBP
target_skeleton=&/M_MED_EssayViewPier_Pack_Skeleton
graphs [
 AnimGraph {nodes [
   Input Pose()
   AnimGraphNode_StateMachine_0 = IdleBreak() {is=title AnimGraphNode_StateMachine}
   n1 = "Transform (Modify) Bone - Bone: pack"(ComponentPose=Local To Component(LocalPose=@AnimGraphNode_StateMachine_0), Translation="0,0,0", Rotation=Get backpack_rotate()) {
    rotationMode=BMM_Additive
    rotationSpace=BCS_BoneSpace
    translationMode=BMM_Additive
    translationSpace=BCS_BoneSpace
    alphaBoolBlend {blendOption=Linear}
    boneToModify {boneName=pack}
   }
   n2 = Anim Dynamics - Bone: base_chair!(
    SimulationSpace:AnimPhysSimSpaceType Enum=Component
    ComponentPose=@n1
    Alpha=0.5
    GravityOverride:Vector=Get GravityOverride()
    ExternalForce="0,0,0"
   ) -> (Pose) {
    angularDampingOverride=1
    angularSpringConstant=20000
    bAngularSpring=true
    bOverrideAngularDamping=true
    bUseGravityOverride=true
    gravityScale=0.35
    alphaBoolBlend {blendOption=Linear}
    boundBone {boneName=base_chair}
    componentLinearAccScale=<1 1 1>
    componentLinearVelScale=<1 1 1>
    gravityOverride {z=-980}
    physicsBodyDefinitions [{
      boundBone {boneName=base_chair}
      boxExtents=<10 10 10>
      constraintSetup {twistAxis=AxisY, angularLimitsMax=<90 0 0>, angularLimitsMin=<-80 0 0>, angularTarget=<1 0 0>}
      localJointOffset=<0 6 0>
     }]
    relativeSpaceBone {boneName=MainBackpack}
   }
   n3 = Anim Dynamics - Bone: dyn_caster_001!(SimulationSpace:AnimPhysSimSpaceType Enum=BoneRelative, ComponentPose=@n2, Alpha=1, GravityOverride:Vector="0,0,0", ExternalForce="0,0,0") -> (Pose) {
    angularDampingOverride=1
    angularSpringConstant=20000
    bAngularSpring=true
    bOverrideAngularDamping=true
    bUseGravityOverride=true
    gravityScale=0.35
    simulationSpace=BoneRelative
    alphaBoolBlend {blendOption=Linear}
    boundBone {boneName=dyn_caster_001}
    componentLinearAccScale {x=1}
    componentLinearVelScale {x=1}
    gravityOverride {z=-980}
    physicsBodyDefinitions [{
      boundBone {boneName=dyn_caster_001}
      boxExtents=<15 15 15>
      constraintSetup {twistAxis=AxisY, angularLimitsMax=<80 0 0>, angularLimitsMin=<-80 0 0>, angularTarget=<1 0 0>}
      localJointOffset=<2 0 3.7809>
     }]
    relativeSpaceBone {boneName=base_chair}
   }
   n4 = Anim Dynamics - Bone: dyn_caster_002!(SimulationSpace:AnimPhysSimSpaceType Enum=BoneRelative, ComponentPose=@n3, Alpha=1, GravityOverride:Vector="0,0,0", ExternalForce="0,0,0") -> (Pose) {
    angularDampingOverride=1
    angularSpringConstant=20000
    bAngularSpring=true
    bOverrideAngularDamping=true
    bUseGravityOverride=true
    gravityScale=0.35
    simulationSpace=BoneRelative
    alphaBoolBlend {blendOption=Linear}
    boundBone {boneName=dyn_caster_002}
    componentLinearAccScale {x=1}
    componentLinearVelScale {x=1}
    gravityOverride {z=-980}
    physicsBodyDefinitions [{
      boundBone {boneName=dyn_caster_002}
      boxExtents=<15 15 15>
      constraintSetup {twistAxis=AxisY, angularLimitsMax=<80 0 0>, angularLimitsMin=<-80 0 0>, angularTarget=<1 0 0>}
      localJointOffset=<2 0 3.7809>
     }]
    relativeSpaceBone {boneName=base_chair}
   }
   n5 = Anim Dynamics - Bone: dyn_caster_003!(SimulationSpace:AnimPhysSimSpaceType Enum=BoneRelative, ComponentPose=@n4, Alpha=1, GravityOverride:Vector="0,0,0", ExternalForce="0,0,0") -> (Pose) {
    angularDampingOverride=1
    angularSpringConstant=20000
    bAngularSpring=true
    bOverrideAngularDamping=true
    bUseGravityOverride=true
    gravityScale=0.35
    simulationSpace=BoneRelative
    alphaBoolBlend {blendOption=Linear}
    boundBone {boneName=dyn_caster_003}
    componentLinearAccScale {x=1}
    componentLinearVelScale {x=1}
    gravityOverride {z=-980}
    physicsBodyDefinitions [{
      boundBone {boneName=dyn_caster_003}
      boxExtents=<15 15 15>
      constraintSetup {twistAxis=AxisY, angularLimitsMax=<80 0 0>, angularLimitsMin=<-80 0 0>, angularTarget=<1 0 0>}
      localJointOffset=<2 0 3.7809>
     }]
    relativeSpaceBone {boneName=base_chair}
   }
   n6 = Anim Dynamics - Bone: dyn_caster_004!(SimulationSpace:AnimPhysSimSpaceType Enum=BoneRelative, ComponentPose=@n5, Alpha=1, GravityOverride:Vector="0,0,0", ExternalForce="0,0,0") -> (Pose) {
    angularDampingOverride=1
    angularSpringConstant=20000
    bAngularSpring=true
    bOverrideAngularDamping=true
    bUseGravityOverride=true
    gravityScale=0.35
    simulationSpace=BoneRelative
    alphaBoolBlend {blendOption=Linear}
    boundBone {boneName=dyn_caster_004}
    componentLinearAccScale {x=1}
    componentLinearVelScale {x=1}
    gravityOverride {z=-980}
    physicsBodyDefinitions [{
      boundBone {boneName=dyn_caster_004}
      boxExtents=<15 15 15>
      constraintSetup {twistAxis=AxisY, angularLimitsMax=<80 0 0>, angularLimitsMin=<-80 0 0>, angularTarget=<1 0 0>}
      localJointOffset=<2 0 3.7809>
     }]
    relativeSpaceBone {boneName=base_chair}
   }
   n7 = Anim Dynamics - Bone: dyn_caster_005!(SimulationSpace:AnimPhysSimSpaceType Enum=BoneRelative, ComponentPose=@n6, Alpha=1, GravityOverride:Vector="0,0,0", ExternalForce="0,0,0") -> (Pose) {
    angularDampingOverride=1
    angularSpringConstant=20000
    bAngularSpring=true
    bOverrideAngularDamping=true
    bUseGravityOverride=true
    gravityScale=0.35
    simulationSpace=BoneRelative
    alphaBoolBlend {blendOption=Linear}
    boundBone {boneName=dyn_caster_005}
    componentLinearAccScale {x=1}
    componentLinearVelScale {x=1}
    gravityOverride {z=-980}
    physicsBodyDefinitions [{
      boundBone {boneName=dyn_caster_005}
      boxExtents=<15 15 15>
      constraintSetup {twistAxis=AxisY, angularLimitsMax=<80 0 0>, angularLimitsMin=<-80 0 0>, angularTarget=<1 0 0>}
      localJointOffset=<2 0 3.7809>
     }]
    relativeSpaceBone {boneName=base_chair}
   }
   n8 = RigidBody!(ComponentPose=@n7, Alpha, OverrideWorldGravity:Vector="0,0,0", ExternalForce=Get GravityOverride()) -> (Pose) {
    bOverrideWorldGravity=true
    overridePhysicsAsset=&/M_MED_EssayViewPier_Pack_Physics
    simulationSpace=BaseBoneSpace
    alphaBoolBlend {blendOption=Linear}
    baseBoneRef {boneName=root}
    componentAppliedLinearAccClamp=<3000 3000 3000>
    componentLinearAccScale=<0.3 0.2 0.2>
    componentLinearVelScale=<0.3 0.2 0.2>
    simSpaceSettings {
     maxAngularAcceleration=1000
     maxAngularVelocity=3
     maxLinearAcceleration=700
     maxLinearVelocity=500
     velocityScaleZ=0
     worldAlpha=0.2
     externalLinearDragV=<0.2 0.2>
    }
   }
   Output Pose(Result=Component To Local(ComponentPose=@n8)) {layerGroup=None}
   Get backpack_translation()
  ]}
 IdleBreak {
  entry=IdlePose
  path=AnimGraph.AnimGraphNode_StateMachine_0.IdleBreak
  ids=id
  nodes [
   AnimStateNode_0 = IdlePose() {is=title AnimStateNode}
   AnimStateNode_1 = IdleBreak() {is=title AnimStateNode}
   AnimStateTransitionNode_0 = AnimStateTransitionNode()
   AnimStateTransitionNode_1 = AnimStateTransitionNode() {bAutomaticRuleBasedOnSequencePlayerInState=true}
  ]
  links [
   {
    dst=IdlePose
    rule=AnimGraph.AnimGraphNode_StateMachine_0.IdleBreak.AnimStateTransitionNode_1.Transition
    src=IdleBreak
    via=AnimStateTransitionNode_1
   }
   {
    dst=IdleBreak
    rule=AnimGraph.AnimGraphNode_StateMachine_0.IdleBreak.AnimStateTransitionNode_0.Transition
    src=IdlePose
    via=AnimStateTransitionNode_0
   }
  ]
 }
 IdlePose {
  path=AnimGraph.AnimGraphNode_StateMachine_0.IdleBreak.AnimStateNode_0.IdlePose
  nodes [
   n1 = Sequence Player 'EssayViewPier_Idle_PACK'() {bLoopAnimation=false, sequence=/BRCosmetics/Accessories/FORT_Backpacks/Backpack_M_MED_EssayViewPier/Animation/EssayViewPier_Idle_PACK}
   Output Animation Pose(Result=@n1)
  ]
 }
 IdleBreak {
  path=AnimGraph.AnimGraphNode_StateMachine_0.IdleBreak.AnimStateNode_1.IdleBreak
  nodes [
   n1 = Sequence Player 'EssayViewPier_IdleBreak_PACK'() {bLoopAnimation=false, sequence=/BRCosmetics/Accessories/FORT_Backpacks/Backpack_M_MED_EssayViewPier/Animation/EssayViewPier_IdleBreak_PACK}
   Output Animation Pose(Result=@n1)
  ]
 }
 Transition {
  path=AnimGraph.AnimGraphNode_StateMachine_0.IdleBreak.AnimStateTransitionNode_0.Transition
  nodes [Result(bCanEnterTransition=Get Play_IdleBreak() -> (Play_IdleBreak:bool))]
 }
 Transition {path=AnimGraph.AnimGraphNode_StateMachine_0.IdleBreak.AnimStateTransitionNode_1.Transition, nodes [Result()]}
 EventGraph {nodes [
   Event BlueprintUpdateAnimation()
   > n1 = Sequence!(execute) -> (then_0, then_1, then_2:Exec, then_3:Exec)
   > n2 = Gravity Override State Function DEPRECATED(CharacterPartAnimInstance=Self-Reference(), GameTimeInSeconds=GetGameTimeInSeconds(), GravityOverrideParams=Get Gravity Override Params())
   > Set GravityOverride(GravityOverride=@n2)
   n3 = GetOwnerSkeletalMeshComponent()
   n4 = BreakRotator(InRot=BreakTransform(InTransform=GetDeltaTransformFromRefPose(self=@n3, BoneName=spine_05, BaseName=spine_01)).Rotation)
   Set backpack_rotate(execute=@n1.then_1, backpack_rotate=MakeRotator(Roll=MapRangeClamped(Value=@n4.Yaw, InRangeA=0, InRangeB=-30, OutRangeB=45)))
   > Set backpack_translation(backpack_translation=MakeVector(Y=MapRangeClamped(Value=@n4.Yaw, InRangeA=0, InRangeB=-30, OutRangeB=-2)))
   n5 = Branch(execute=@n1.then_2, Condition=Get bIsInFrontEnd())
   > Set SimlationSpace(SimlationSpace:AnimPhysSimSpaceType Enum=World) -> (Output_Get:AnimPhysSimSpaceType Enum)
   > Set AlphaFE(AlphaFE:double=0.15) -> (Output_Get:double)
   Set SimlationSpace(execute=@n5.else)
   > Set AlphaFE(AlphaFE=1)
   Delay(execute=@n1.then_3, Duration=RandomFloatInRange(Min=15, Max=30))
   > Set Play_IdleBreak(Play_IdleBreak:bool=true) -> (Output_Get:bool)
   > Delay(Duration=0.1)
   > Set Play_IdleBreak(Play_IdleBreak=false)
   TryGetPawnOwner()
  ]}
]
variables [
 AlphaFE {type=number}
 backpack_rotate {type=Rotator}
 backpack_translation {type=Vector}
 Gravity Override Params {type=GravityOverrideParamsStruct, default {onGround_Standing {windFrequency=<500 500 500>}}}
 GravityOverride {type=Vector}
 Play_IdleBreak {type=boolean}
 SimlationSpace {type=AnimPhysSimSpaceType}
]
